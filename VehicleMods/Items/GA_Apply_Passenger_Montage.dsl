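; Gameplay ability from VehicleMods that plays the passenger turret montage, apparently on a player in a vehicle turret seat.
; On activation it waits for a client-server net sync, caches the avatar as a FortPlayerPawn, then starts the montage via a play-and-wait task.
; Montage completion ends the ability on authority only; interruption and cancellation are unhandled.
; Debug prints trace playback and ability end.
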(event Ability|EventActivateAbility
  (Ability|Tasks|WaitNetSync "BothWait"
    (:OnSync
      (bind _asfort_player_pawn (Utilities|Casting|CastToFortPlayerPawn (Ability|GetAvatarActorfromActorInfo)))
      (Variables|Default|SetPlayer%20Pawn _asfort_player_pawn)
      (Ability|Tasks|PlayMontageAndWaitNotify "None" (Variables|Default|GetPassengerTurretMontage)
        (:then
          (Development|PrintString "Passenger Turret Montage Playing"))
        (:OnCompleted
          (Ability|HasAuthority
            (:True
              (Ability|EndAbility))))
        (:OnInterrupted
          _)
        (:OnCancelled
          _)
        (:OnNotifyEnd
          _))
      (Development|PrintString "Passenger Turret Montage Ended"))))

(event Ability|EventOnEndAbility
  (params (bWasCancelled :type Boolean))
  (Development|PrintString "Ability End")
  (Development|PrintString "Passenger Turret Montage Ended"))
