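; Gameplay ability for the mortar vehicle mod that records spent ammo on the vehicle itself, using gameplay effect stacks as the counter.
; On activation it counts tracker effects on the vehicle's ability system component and consumes that much ammo from the pawn's current weapon.
; It then binds to the weapon's consume-ammo event, applying one tracker effect per event regardless of quantity.
; Ability end has no logic, so the delegate is never explicitly unbound.
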(event Ability|EventActivateAbility
  (bind _ge_ammotracker (Variables|Default|GetGE_AmmoTracker))
  ;[0 1 0.2292] Check amount of GEs are on the vehicle and consume ammo
  (|CallParentFunction)
  (bind _asfort_player_pawn (Utilities|Casting|CastToFortPlayerPawn (Ability|GetAvatarActorfromActorInfo)))
  (if (not (== (GameplayEffects|GetGameplayEffectCount (Vehicle|GetVehicleAbilitySystemComponent (Vehicle|GetVehicle _asfort_player_pawn)) _ge_ammotracker 0 true) 0))
    (FortWeapon|Ammo|ConsumeAmmo (Variables|/Script/FortniteGame.FortPawn|GetCurrentWeapon _asfort_player_pawn) (GameplayEffects|GetGameplayEffectCount (Vehicle|GetVehicleAbilitySystemComponent (Vehicle|GetVehicle _asfort_player_pawn)) _ge_ammotracker 0 true) false))
  (EventDispatchers|BindEventtoOnConsumeAmmo (Variables|/Script/FortniteGame.FortPawn|GetCurrentWeapon _asfort_player_pawn) (delegate Custom|WeaponConsumeAmmo)))

(event Custom|WeaponConsumeAmmo
  (params (Quantity :type Integer))
  ;[0 1 0.2292] Check amount of GEs are on the vehicle and consume ammo
  (GameplayEffects|ApplyGameplayEffectToSelf :self (Vehicle|GetVehicleAbilitySystemComponent (Vehicle|GetVehicle (stale Ability|EventActivateAbility _asfort_player_pawn))) :GameplayEffectClass (Variables|Default|GetGE_AmmoTracker) :Level 0.0) ; Apply a Ammo Tracker GE to the vehicle every time we consume ammo
)

(event Ability|EventOnEndAbility
  (params (bWasCancelled :type Boolean)))
