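; Passive gameplay ability for an equipped turret vehicle mod that adjusts the mobile HUD and build state.
; On activation it calls the parent, then on mobile hides the tagged HUD elements under a "TurretEquip" context and forces the owning player out of build mode.
; When the ability ends on mobile, it restores the same elements using that context.
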(event Ability|EventActivateAbility
  (bind _returnvalue (Ability|GetOwningActorfromActorInfo))
  (bind _returnvalue_2 (Fortnite|GetFortPlayerControllerfromActor _returnvalue))
  ;[0 1 0.2292] Disable HUD elements by tag.
  (|CallParentFunction)
  (if (TouchPlatform|IsMobileGame)
    (HUD|SetHUDElementVisibilitywithContext _returnvalue_2 (Variables|Default|GetHUDTagsToHide) true "TurretEquip" "")
    ;[0 0.4316 1] Exit build mode when entering
    (if (QuickBars|IsinBuildMode _returnvalue_2)
      (Class|FortPlayerController|ForceExitBuildMode _returnvalue_2))))

(event Ability|EventOnEndAbility
  (params (bWasCancelled :type Boolean))
  ;[1 0.237 0.9724] Re-enable the elements when ability is removed.
  (if (TouchPlatform|IsMobileGame)
    (HUD|SetHUDElementVisibilitywithContext (Fortnite|GetFortPlayerControllerfromActor (Ability|GetOwningActorfromActorInfo)) (Variables|Default|GetHUDTagsToHide) false "TurretEquip" "")))
